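# Post-process AnimBP for the FlameBride female head that adjusts facial bones on the shared Fortnite 3L head skeleton.
# A chain of constraints ties four center and left forehead bones fully, and two left cheek bones at half alpha, to the head bone.
# A bool blend selects between this and the cached input pose.
# Customization-load and update events set Is FlameBrade from the FlameBride meta tag, though no link to the blend is shown.
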AnimBlueprint /BRCosmetics/Characters/Player/Female/Medium/Heads/F_MED_FlameBride_Head/Meshes/F_MED_FlameBride_PostProcess_AnimBP
target_skeleton=/Game/Characters/Player/Common/Fortnite_Base_Head/Fortnite_Base_3L_Head_Skeleton
graphs [
 AnimGraph {nodes [
   n1 = "Constraint - FACIAL_C_Forehead (P)"(ComponentPose=Local To Component(LocalPose=Use cached pose 'Is FlameBride'()), Alpha=1, ConstraintWeights_0=1) {
    alphaBoolBlend {blendOption=Linear}
    boneToModify {boneName=FACIAL_C_Forehead}
    constraintSetup [{targetBone {boneName=head}}]
    constraintWeights [1]
   }
   n2 = "Constraint - FACIAL_L_ForeheadIn (P)"(ComponentPose=@n1, Alpha=1, ConstraintWeights_0=1) {
    alphaBoolBlend {blendOption=Linear}
    boneToModify {boneName=FACIAL_L_ForeheadIn}
    constraintSetup [{targetBone {boneName=head}}]
    constraintWeights [1]
   }
   n3 = "Constraint - FACIAL_L_ForeheadMid (P)"(ComponentPose=@n2, Alpha=1, ConstraintWeights_0=1) {
    alphaBoolBlend {blendOption=Linear}
    boneToModify {boneName=FACIAL_L_ForeheadMid}
    constraintSetup [{targetBone {boneName=head}}]
    constraintWeights [1]
   }
   n4 = "Constraint - FACIAL_L_ForeheadOut (P)"(ComponentPose=@n3, Alpha=1, ConstraintWeights_0=1) {
    alphaBoolBlend {blendOption=Linear}
    boneToModify {boneName=FACIAL_L_ForeheadOut}
    constraintSetup [{targetBone {boneName=head}}]
    constraintWeights [1]
   }
   n5 = "Constraint - FACIAL_L_CheekInner (P)"(ComponentPose=@n4, Alpha=0.5, ConstraintWeights_0=1) {
    alphaBoolBlend {blendOption=Linear}
    boneToModify {boneName=FACIAL_L_CheekInner}
    constraintSetup [{targetBone {boneName=head}}]
    constraintWeights [1]
   }
   n6 = "Constraint - FACIAL_L_CheekOuter (P)"(ComponentPose=@n5, Alpha=0.5, ConstraintWeights_0=1) {
    alphaBoolBlend {blendOption=Linear}
    boneToModify {boneName=FACIAL_L_CheekOuter}
    constraintSetup [{targetBone {boneName=head}}]
    constraintWeights [1]
   }
   Output Pose(Result=Blend Poses by bool(BlendPose_0=Component To Local(ComponentPose=@n6), BlendPose_1=Use cached pose 'Is FlameBride'()) {blendType=Linear}) {layerGroup=None}
   Save cached pose 'Is FlameBride'(Pose=Input Pose()) {cacheName=Is FlameBride}
  ]}
 EventGraph {nodes [
   n1 = Event OnCharacterCustomizationLoaded()
   Set Is FlameBrade(
    execute=[@n1.then, @n2.then]
    Is FlameBrade:bool=CharacterItemDefinitionHasAnyMetaTag(Tags=Make Array("[0]"=(TagName="Cosmetics.MetaTags.FlameBride")))
   ) -> (Output_Get:bool)
   n2 = Event BlueprintUpdateAnimation()
  ]}
]
variables [
 Is FlameBrade {type=boolean}
 NewVar {type=boolean}
]
